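# GenericHMDMaterial is a Default Lit material in the VR Editor's generic device content, named for a generic head-mounted display.
# The graph holds only two scalar constant nodes, with no textures, parameters or math.
# The file does not show which material inputs those constants feed.
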Material /Engine/VREditor/Devices/Generic/GenericHMDMaterial
graph {
 properties {pixelDepthOffsetMode=PDOM_Legacy, usedShadingModels=Default Lit}
 nodes [
  Constant() {r=0.0025}
  Constant() {r=0.6}
 ]
}
